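# Animation Blueprint for the CampusSire backpack, simulating its three dangling string bones with spring-driven AnimDynamics.
# Each update derives onLand and inSky from skydive flags, blending between looser ground and tighter airborne simulations on the first two strings.
# Additive bone offsets then tilt the pack while skydiving and re-angle those strings while floating.
# bIsInFrontEnd switches output to a separate always-on simulation chain without those offsets.
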AnimBlueprint /BRCosmetics/Accessories/FORT_Backpacks/Backpack_M_LRG_CampusSire/Meshes/M_LRG_CampusSire_Pack_AnimBP
target_skeleton=&/M_LRG_CampusSire_Pack_Skeleton
graphs [
 AnimGraph {nodes [
   n1 = Anim Dynamics - Bone: dyn_string_00!(
    ComponentPose=Local To Component(LocalPose=Use cached pose 'input pose'())
    bAlphaBoolEnabled:bool=Get onLand() -> (onLand:bool)
    ExternalForce="0,0,0"
   ) -> (Pose) {
    alphaInputType=Bool
    angularDampingOverride=0.9999
    angularSpringConstant=1000
    bAngularSpring=true
    bOverrideAngularDamping=true
    bUseGravityOverride=true
    alphaBoolBlend {blendInTime=0.2, blendOption=Linear, blendOutTime=0.2}
    boundBone {boneName=dyn_string_00}
    componentAppliedLinearAccClamp=<1000 1000 1000>
    componentLinearAccScale=<0.3 0.3 0.3>
    componentLinearVelScale=<0.4 0.4 0.4>
    gravityOverride {z=-980}
    physicsBodyDefinitions [{
      boundBone {boneName=dyn_string_00}
      boxExtents=<5 5 5>
      constraintSetup {angularLimitsMax=<0 50 20>, angularLimitsMin=<0 -5 -40>, angularTarget=<0 0 0>}
      localJointOffset=<1 0 0>
     }]
   }
   n2 = Anim Dynamics - Bone: dyn_string_00(ComponentPose=@n1, bAlphaBoolEnabled=Get inSky() -> (inSky:bool), ExternalForce="0,0,0") {
    alphaInputType=Bool
    angularDampingOverride=0.9999
    angularSpringConstant=800
    bAngularSpring=true
    bOverrideAngularDamping=true
    bUseGravityOverride=true
    alphaBoolBlend {blendInTime=0.2, blendOption=Linear, blendOutTime=0.2}
    boundBone {boneName=dyn_string_00}
    componentAppliedLinearAccClamp=<800 800 800>
    componentLinearAccScale=<0.2 0.2 0.2>
    componentLinearVelScale=<0.3 0.3 0.3>
    gravityOverride {z=-600}
    physicsBodyDefinitions [{
      boundBone {boneName=dyn_string_00}
      boxExtents=<5 5 5>
      constraintSetup {angularLimitsMax=<0 10 20>, angularLimitsMin=<0 0 -10>, angularTarget=<0 0 0>}
      localJointOffset=<0.5 0 0>
     }]
   }
   n3 = Anim Dynamics - Bone: dyn_string_01!(ComponentPose=@n2, bAlphaBoolEnabled:bool=Get onLand(), ExternalForce="0,0,0") -> (Pose) {
    alphaInputType=Bool
    angularDampingOverride=0.9999
    angularSpringConstant=1000
    bAngularSpring=true
    bOverrideAngularDamping=true
    bUseGravityOverride=true
    alphaBoolBlend {blendInTime=0.2, blendOption=Linear, blendOutTime=0.2}
    boundBone {boneName=dyn_string_01}
    componentAppliedLinearAccClamp=<1000 1000 1000>
    componentLinearAccScale=<0.4 0.4 0.4>
    componentLinearVelScale=<0.5 0.5 0.5>
    gravityOverride {z=-980}
    physicsBodyDefinitions [{
      boundBone {boneName=dyn_string_01}
      boxExtents=<5 5 5>
      constraintSetup {angularLimitsMax=<0 60 20>, angularLimitsMin=<0 -5 -40>, angularTarget=<0 0 0>}
      localJointOffset=<1.3 0 0>
     }]
   }
   n4 = Anim Dynamics - Bone: dyn_string_01(ComponentPose=@n3, bAlphaBoolEnabled=Get inSky(), ExternalForce="0,0,0") {
    alphaInputType=Bool
    angularDampingOverride=0.9999
    angularSpringConstant=800
    bAngularSpring=true
    bOverrideAngularDamping=true
    bUseGravityOverride=true
    alphaBoolBlend {blendInTime=0.2, blendOption=Linear, blendOutTime=0.2}
    boundBone {boneName=dyn_string_01}
    componentAppliedLinearAccClamp=<1000 1000 1000>
    componentLinearAccScale=<0.2 0.2 0.2>
    componentLinearVelScale=<0.3 0.3 0.3>
    gravityOverride {z=-700}
    physicsBodyDefinitions [{
      boundBone {boneName=dyn_string_01}
      boxExtents=<5 5 5>
      constraintSetup {angularLimitsMax=<0 10 20>, angularLimitsMin=<0 0 -5>, angularTarget=<0 0 0>}
      localJointOffset=<0.6 0 0>
     }]
   }
   n5 = Anim Dynamics - Bone: dyn_string_02(ComponentPose=@n4, Alpha=1, ExternalForce="0,0,0") {
    angularDampingOverride=0.9999
    angularSpringConstant=1000
    bAngularSpring=true
    bOverrideAngularDamping=true
    bUseGravityOverride=true
    alphaBoolBlend {blendOption=Linear}
    boundBone {boneName=dyn_string_02}
    componentAppliedLinearAccClamp=<1000 1000 1000>
    componentLinearAccScale=<0.4 0.4 0.4>
    componentLinearVelScale=<0.5 0.5 0.5>
    gravityOverride {z=-980}
    physicsBodyDefinitions [{
      boundBone {boneName=dyn_string_02}
      boxExtents=<5 5 5>
      constraintSetup {angularLimitsMax=<10 50 30>, angularLimitsMin=<-10 -20 -40>, angularTarget=<0 0 0>}
      localJointOffset=<1 0 0>
     }]
   }
   n6 = "Transform (Modify) Bone - Bone: pack"!(ComponentPose=@n5, bAlphaBoolEnabled:bool=Get bIsSkyDiving(), Translation="0,1,0", Rotation="0,0,17", Scale) -> (Pose) {
    alphaInputType=Bool
    rotationMode=BMM_Additive
    rotationSpace=BCS_ParentBoneSpace
    translationMode=BMM_Additive
    translationSpace=BCS_ParentBoneSpace
    alphaBoolBlend {blendInTime=0.2, blendOption=Linear, blendOutTime=0.2}
    boneToModify {boneName=pack}
    rotation {roll=17}
    translation {y=1}
   }
   n7 = Get bIsSkydiveFloating()
   n8 = "Transform (Modify) Bone - Bone: dyn_string_00"!(ComponentPose=@n6, bAlphaBoolEnabled:bool=@n7, Translation="0,0,0", Rotation="-50.482,-25.405,78.557", Scale="1,1,1") -> (Pose) {
    alphaInputType=Bool
    rotationMode=BMM_Additive
    rotationSpace=BCS_ParentBoneSpace
    translationMode=BMM_Additive
    translationSpace=BCS_ParentBoneSpace
    alphaBoolBlend {blendInTime=0.2, blendOption=Linear, blendOutTime=0.2}
    boneToModify {boneName=dyn_string_00}
    rotation=<rot -50.482 -25.405 78.557>
   }
   n9 = "Transform (Modify) Bone - Bone: dyn_string_01"!(ComponentPose=@n8, bAlphaBoolEnabled:bool=@n7, Translation="0,0,0", Rotation="-32.873,-62.250,23.395", Scale="1,1,1") -> (Pose) {
    alphaInputType=Bool
    rotationMode=BMM_Additive
    rotationSpace=BCS_ParentBoneSpace
    translationMode=BMM_Additive
    translationSpace=BCS_ParentBoneSpace
    alphaBoolBlend {blendInTime=0.2, blendOption=Linear, blendOutTime=0.2}
    boneToModify {boneName=dyn_string_01}
    rotation=<rot -32.873 -62.25 23.395>
   }
   n10 = Anim Dynamics - Bone: dyn_string_00!(ComponentPose=Local To Component(LocalPose=Use cached pose 'input pose'()), Alpha:float=1, ExternalForce="0,0,0") -> (Pose) {
    angularDampingOverride=1
    angularSpringConstant=1000
    bAngularSpring=true
    bOverrideAngularDamping=true
    bUseGravityOverride=true
    alphaBoolBlend {blendInTime=0.2, blendOption=Linear, blendOutTime=0.2}
    boundBone {boneName=dyn_string_00}
    componentAppliedLinearAccClamp=<1000 1000 1000>
    componentLinearAccScale=<0.2 0.2 0.2>
    componentLinearVelScale=<0.3 0.3 0.3>
    gravityOverride {z=-980}
    physicsBodyDefinitions [{
      boundBone {boneName=dyn_string_00}
      boxExtents=<6 6 6>
      constraintSetup {angularLimitsMax=<0 50 20>, angularLimitsMin=<0 -5 -40>, angularTarget=<0 0 0>}
      localJointOffset=<1 0 0>
     }]
   }
   n11 = Anim Dynamics - Bone: dyn_string_01!(ComponentPose=@n10, Alpha:float=1, ExternalForce="0,0,0") -> (Pose) {
    angularDampingOverride=1
    angularSpringConstant=1000
    bAngularSpring=true
    bOverrideAngularDamping=true
    bUseGravityOverride=true
    alphaBoolBlend {blendInTime=0.2, blendOption=Linear, blendOutTime=0.2}
    boundBone {boneName=dyn_string_01}
    componentAppliedLinearAccClamp=<1000 1000 1000>
    componentLinearAccScale=<0.3 0.3 0.3>
    componentLinearVelScale=<0.4 0.4 0.4>
    gravityOverride {z=-980}
    physicsBodyDefinitions [{
      boundBone {boneName=dyn_string_01}
      boxExtents=<5 5 5>
      constraintSetup {angularLimitsMax=<0 60 20>, angularLimitsMin=<0 -5 -40>, angularTarget=<0 0 0>}
      localJointOffset=<1 0 0>
     }]
   }
   n12 = Anim Dynamics - Bone: dyn_string_02(ComponentPose=@n11, Alpha=1, ExternalForce="0,0,0") {
    angularDampingOverride=1
    angularSpringConstant=1000
    bAngularSpring=true
    bOverrideAngularDamping=true
    bUseGravityOverride=true
    alphaBoolBlend {blendOption=Linear}
    boundBone {boneName=dyn_string_02}
    componentAppliedLinearAccClamp=<1000 1000 1000>
    componentLinearAccScale=<0.3 0.3 0.3>
    componentLinearVelScale=<0.4 0.4 0.4>
    gravityOverride {z=-980}
    physicsBodyDefinitions [{
      boundBone {boneName=dyn_string_02}
      boxExtents=<4 4 4>
      constraintSetup {angularLimitsMax=<10 50 30>, angularLimitsMin=<-10 -20 -40>, angularTarget=<0 0 0>}
      localJointOffset=<1 0 0>
     }]
   }
   Output Pose(Result=Blend Poses by bool(BlendPose_0=Component To Local(ComponentPose=@n12), BlendPose_1=Component To Local(ComponentPose=@n9), bActiveValue=Get bIsInFrontEnd()) {blendType=Linear}) {layerGroup=None}
   Save cached pose 'input pose'(Pose=Input Pose()) {cacheName=input pose}
  ]}
 EventGraph {nodes [
   Event BlueprintUpdateAnimation()
   > n1 = Branch(Condition=OR Boolean!(A=Get bIsSkydiveFloating(), B=Get bIsSkydiveDiveMode(), C:bool=Get bIsSkyDiving()) -> (ReturnValue))
   > Set onLand(onLand:bool=false) -> (Output_Get:bool)
   > Set inSky(inSky:bool=true) -> (Output_Get:bool)
   Set inSky(execute=@n1.else, inSky=false)
   > Set onLand(onLand=true)
   TryGetPawnOwner()
  ]}
]
variables [
 inSky {type=boolean}
 On Land {type=boolean}
 onLand {type=boolean}
]
